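; TeamObjective_Food_CT1 is a team objective Blueprint for the Barrier playlist, kept in its ForCapture folder.
; The construction script gets the child actor's class and casts it to the TeamGeneratorCT1 class, but nothing uses the result.
; BeginPlay is empty, so the graphs shown add no gameplay behavior of their own.
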(event EventBeginPlay)

(fn ConstructionScript
  (Utilities|Casting|CastToTeamGeneratorCT1Class (Utilities|GetClass (Variables|Default|GetChildActor))))
